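# M_Axis is a Default Lit material in the effects debug folder, apparently meant for axis meshes.
# It is flagged for use on instanced static meshes and Niagara mesh particles.
# Its graph holds only a vertex color scaled down by a constant multiply, with no connection to a material output shown.
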Material /Game/Effects/Fort_Effects/Materials/Debug/M_Axis
graph {
 properties {bUsedWithInstancedStaticMeshes=true, bUsedWithNiagaraMeshParticles=true, pixelDepthOffsetMode=PDOM_Legacy, usedShadingModels=Default Lit}
 nodes [Multiply(A=VertexColor()) {constB=0.2}]
}
