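# M_Debug_MeshUVs is a two-sided debug material for Niagara mesh particles that visualizes mesh UV layout.
# UVs fill red and green, and TwoSidedSign fills blue on back faces.
# Procedural edge masks from distance to UV center multiply in dark lines along the 0-1 border and a 4x4 tiled grid.
# A half-strength Fresnel term is also computed, though the graph does not show which material inputs either result feeds.
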Material /Game/Effects/Fort_Effects/Materials/Debug/M_Debug_MeshUVs
graph {
 properties {bAutomaticallySetUsageInEditor=false, bUsedWithNiagaraMeshParticles=true, twoSided=true, usedShadingModels=Default Lit}
 nodes [
  Multiply(A=Fresnel() {baseReflectFraction=0}) {constB=0.5}
  n1 = TextureCoordinate()
  n2 = OneMinus(None=Saturate(None=TwoSidedSign()))
  n3 = TextureCoordinate()
  n4 = Multiply(A=Abs(None=Subtract(A=@n3) {constB=0.5})) {constB=2}
  n5 = OneMinus(None=InverseLinearInterpolate(Value=OneMinus(None=@n4)) {constA=0.0299, constB=0.03})
  n6 = ComponentMask(None=@n5) {r=true}
  n7 = ComponentMask(None=@n5) {g=true}
  n8 = Saturate(None=OneMinus(None=Max(A=@n6, B=@n7)))
  n9 = Frac(None=TextureCoordinate() {uTiling=4, vTiling=4})
  n10 = Multiply(A=Abs(None=Subtract(A=@n9) {constB=0.5})) {constB=2}
  n11 = OneMinus(None=InverseLinearInterpolate(Value=OneMinus(None=@n10)) {constA=0.04, constB=0.05})
  n12 = ComponentMask(None=@n11) {r=true}
  n13 = ComponentMask(None=@n11) {g=true}
  n14 = Saturate(None=OneMinus(None=Max(A=@n12, B=@n13)))
  Reroute(None=Multiply(A=AppendVector(A=@n1, B=@n2), B=Multiply(A=@n8, B=@n14)))
 ]
}
