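# M_SimpleMat is a minimal Default Lit material in the effects Debug folder, flagged for use with Nanite.
# Its graph holds a single Color vector parameter with a mid-grey default.
# No connection from that parameter to a material output appears in the graph.
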Material /Game/Effects/Fort_Effects/Materials/Debug/M_SimpleMat
graph {
 properties {bUsedWithNanite=true, pixelDepthOffsetMode=PDOM_Legacy, usedShadingModels=Default Lit}
 nodes [VectorParameter() {parameterName=Color, defaultValue=<rgba 0.5 0.5 0.5 1>}]
}
